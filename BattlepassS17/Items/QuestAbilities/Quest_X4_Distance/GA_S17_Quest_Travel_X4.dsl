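; Season 17 quest gameplay ability that tracks how far the player travels in the X-4 biplane.
; On activation it seeds the pawn's location and starts a one-second looping timer, with a final update on ability end.
; Each update requires a driver or passenger tag and the vehicle's biplane tag, then sends floored horizontal meters to the quest manager as a Distance.Biplane stat.
; Ineligible updates reset the baseline location.
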(event Ability|EventActivateAbilityFromEvent
  (params (EventData :type Gameplay%20Event%20Data%20Structure%20%28by%20ref%29))
  ;[1 1 1] Activate ability
  (|CallParentFunction EventData)
  (Ability|Tasks|WaitGameplayTagAdd (Variables|GAVars|Setup|GetT_Driver) 0 true
    (:then
      (Ability|Tasks|WaitGameplayTagAdd (Variables|GAVars|Setup|GetT_Passenger) 0 true
        (:Added
          _)))
    (:Added
      _))
  (bind _output_get (Variables|GAVars|Data|SetPreviousLocation (Transformation|GetActorLocation (Variables|GAVars|Data|GetOwningPlayerPawn))))
  (Variables|GAVars|Data|SetCurrentLocation _output_get)
  (Utilities|Time|SetTimerbyEvent (delegate Custom|SendDistanceUpdate) 1.0 true))

(event Custom|SendDistanceUpdate
  (bind (_case_0_true _case_1_true) (GameplayTags|TagContainer|CompareTagAssetInterfacetoOtherTags (Variables|GAVars|Setup|GetT_Passenger) (Variables|GAVars|Setup|GetT_Driver) (Variables|GAVars|Data|GetOwningPlayerPawn)))
  (bind (_previouslocation_x _previouslocation_y) (Math|Vector|BreakVector (Variables|GetPreviousLocation_X)))
  (bind (_currentlocation_x _currentlocation_y) (Math|Vector|BreakVector (Variables|GetCurrentLocation_X)))
  ;[1 1 1] Send Distance Update
  (Variables|GAVars|Data|SetCurrentLocation (Transformation|GetActorLocation (Variables|GAVars|Data|GetOwningPlayerPawn)))
  ;;[1 1 1] Check if driver/passenger
  (if (or _case_0_true _case_1_true)
    ;;[1 1 1] Check if vehicle is Biplane
    (if (GameplayTags|HasAnyMatchingGameplayTags (Vehicle|GetVehicleAbilitySystemComponent (Vehicle|GetVehicleInterface (Variables|GAVars|Data|GetOwningPlayerPawn))) (Variables|GAVars|Setup|GetT_Biplane))
      ;;[1 1 1] Cache distance traveled (horizontal only)
      (Variables|GAVars|Data|SetMetersTraveled (Math|Float|Floor (/ (Math|Vector|VectorLength (Math|Vector|vector-vector (Math|Vector|MakeVector _currentlocation_x _currentlocation_y) (Math|Vector|MakeVector _previouslocation_x _previouslocation_y))) 1e+02)))
      (Variables|GAVars|Data|SetPreviousLocation (Variables|GAVars|Data|GetCurrentLocation)) ; Reset location
      ;;[1 1 1] Send Quest Progress
      (if (> (Variables|GAVars|Data|GetMetersTraveled) 0)
        (Quest|SendComplexCustomStatEvent (Variables|GAVars|Data|GetOwningPlayerQuestManager) 0 "(GameplayTags=((TagName=\"Athena.Quests.Distance.Biplane\")))" "(GameplayTags=)" (Variables|GAVars|Data|GetMetersTraveled)))
      (else
        (bind _output_get (Variables|GAVars|Data|SetPreviousLocation (Transformation|GetActorLocation (Variables|GAVars|Data|GetOwningPlayerPawn))))
        (Variables|GAVars|Data|SetCurrentLocation _output_get) ; Reset prev and curr locations while player remains in a vehicle
      ))
    (else
      (Variables|GAVars|Data|SetPreviousLocation (Variables|GAVars|Data|GetCurrentLocation)))))

(event Ability|EventOnEndAbility
  (params (bWasCancelled :type Boolean))
  ;[1 1 1] On End Ability
  (|CallParentFunction bWasCancelled)
  (CallFunction|SendDistanceUpdate))
